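# Idle glow Niagara system for the HonorBraceJoust male medium body cosmetic.
# Four burst-spawned emitters (body mesh glow, sprite overlay, two glow sprite layers) follow sockets through a shared Socket Reader, with partition-indexed arrays placing glow sprites.
# A system Timeline drives alpha on all emitters, while front-end modifiers, quality level and head/hand-hidden flags adjust individual layers.
# GlowOverlay also passes particle positions to dynamic material parameters.
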NiagaraSystem /BRCosmetics/Characters/Player/Male/Medium/Bodies/M_MED_HonorBraceJoust/FX/NS_Character_HonorBraceJoust_Idle
dependencies {
 usedDataInterfaces [
  &NiagaraDataInterfaceCurve
  &NiagaraDataInterfaceArrayColor
  &NiagaraDataInterfaceArrayFloat2
  &NiagaraDataInterfaceArrayFloat
  &NiagaraDataInterfaceArrayInt32
  &NiagaraDataInterfaceArrayFloat3
  &NiagaraDataInterfaceSocketReader
  &NiagaraDataInterfaceColorCurve
 ]
 usedDynamicInputs [
  &InvertBool
  &Boolean_AndOperation
  &SetBoolByFloatComparison
  &ColorFromCurve
 ]
 usedModules [
  &self:SystemSpawnScript.SystemScriptSource.SystemScriptGraph.NiagaraNodeAssignment_2.SetVariables_#1
  &self:SystemSpawnScript.SystemScriptSource.SystemScriptGraph.NiagaraNodeAssignment_4.SetVariables_#2
  &SystemState
  &Timeline
  &self:SystemSpawnScript.SystemScriptSource.SystemScriptGraph.NiagaraNodeAssignment_7.SetVariables_#3
  &self:SystemSpawnScript.SystemScriptSource.SystemScriptGraph.NiagaraNodeAssignment_1.SetVariables_#4
  &self:SystemSpawnScript.SystemScriptSource.SystemScriptGraph.NiagaraNodeAssignment_3.SetVariables_#5
 ]
}
emitters {
 Em_Body_Glow {
  emitterUpdateScript [
   &EmitterState(
    Inactive Response=::NewEnumerator1
    Life Cycle Mode=::NewEnumerator1
    Loop Behavior=::NewEnumerator1
    Loop Duration Mode=::NewEnumerator1
    UseLoopDelay=false
   )
   &SpawnBurst_Instantaneous(Use Spawn Probability=false, UseLoopCountLimit=false)
  ]
  particleSpawnScript [&InitializeParticle(
    Color Mode=::NewEnumerator0
    Mesh Scale=&FN_FrontEndVector3Modifier(Gameplay=<2.25 1.5 1.5>, Front End=<2 2 2>)
    Mesh Scale Mode=::NewEnumerator4
    UsePositionOffset=false
    Write Material Random=true
   )]
  particleUpdateScript [
   &ParticleState(Kill Particles When Lifetime Has Elapsed=false, Let Infinitely Lived Particles Die When Emitter Deactivates=true)
   &SocketLocation(
    SamplePosOffset=true
    Sampled Position Offset=<-10 0 0>
    Socket Index=$Particles.PartitionParticles.Partition
    Socket Reader=$System.Socket Reader
    Socket Sampling Mode=::NewEnumerator3
    VelocitySpeedLimit=false
   )
   &CameraOffset()
   &ScaleColor(Scale Alpha=$System.Timeline001.Float, ScaleA=true, ScaleRGB=true) {module=ScaleColor002}
   /SetVariables(Particles.MeshOrientationFE:Quat4f=&LerpQuaternion(
     Quaternion A:Quat4f=$Particles.MeshOrientation
     Quaternion B:Quat4f=<0 1 0 1>
     Lerp Factor:NiagaraFloat=1
     Enable Spherical Linear Interpolation:NiagaraBool=false
    )) {isSetParameters=true}
  ]
  rendererClasses [&NiagaraMeshRendererProperties]
  renderers [
   {rendererClass=&NiagaraMeshRendererProperties, rendererIndex=0}
   {rendererClass=&NiagaraMeshRendererProperties, rendererIndex=1}
  ]
 }
 GlowOverlay {
  emitterUpdateScript [
   &EmitterState(
    Inactive Response=::NewEnumerator1
    Life Cycle Mode=::NewEnumerator1
    Loop Behavior=::NewEnumerator1
    Loop Duration Mode=::NewEnumerator1
    UseLoopDelay=false
   )
   &SpawnBurst_Instantaneous(Spawn Count=$System.Groups, Use Spawn Probability=false, UseLoopCountLimit=false)
  ]
  particleSpawnScript [
   &PartitionParticles(ID to Partition:NiagaraInt32=&ReturnExecIndex(), Number Of Partitions=$System.Groups)
   &InitializeParticle(
    Color=<rgba 15 0 10.4666 1>
    Lifetime=0.5
    Sprite Size Mode=::NewEnumerator3
    Uniform Sprite Size=55
    UsePositionOffset=false
    Write Material Random=true
   )
  ]
  particleUpdateScript [
   &ParticleState(Kill Particles When Lifetime Has Elapsed=false, Let Infinitely Lived Particles Die When Emitter Deactivates=true)
   &SocketLocation(
    SamplePosOffset=true
    Socket Index=$Particles.PartitionParticles.Partition
    Socket Reader=$System.Socket Reader
    Socket Sampling Mode=::NewEnumerator3
    VelocitySpeedLimit=false
   )
   &CameraOffset(Camera Offset Amount=&FN_FrontEndFloatModifier(Gameplay=20, Front End=0))
   /SetVariables(Particles.OffsetPosition:NiagaraPosition=&AddVectorToPosition(
     Position=$Particles.Position
     Vector=&RotateVectorByQuaternion(Vector To Rotate=<0 0 0>, Quaternion=$Output.SocketLocation.SampledRotation)
    )) {isSetParameters=true}
   &DynamicMaterialParameters(
    Index 0 Param 1=&MakeFloatFromVector(Vector=Linked() {linkedVariable {name=Particles.Position, type=NiagaraPosition}}, Channel=::NewEnumerator0)
    Index 0 Param 2=&MakeFloatFromVector(Vector=Linked() {linkedVariable {name=Particles.Position, type=NiagaraPosition}}, Channel=::NewEnumerator1)
    Index 0 Param 3=&MakeFloatFromVector(Vector=Linked() {linkedVariable {name=Particles.Position, type=NiagaraPosition}}, Channel=::NewEnumerator2)
    Index 1 Param 1=&MakeFloatFromVector(Vector=Linked() {linkedVariable {name=Particles.OffsetPosition, type=NiagaraPosition}}, Channel=::NewEnumerator0)
    Index 1 Param 2=&MakeFloatFromVector(Vector=Linked() {linkedVariable {name=Particles.OffsetPosition, type=NiagaraPosition}}, Channel=::NewEnumerator1)
    Index 1 Param 3=&MakeFloatFromVector(Vector=Linked() {linkedVariable {name=Particles.OffsetPosition, type=NiagaraPosition}}, Channel=::NewEnumerator2)
    Index 1 Param 4=1
    Module.Index0Float0Write=true
    Module.Index0Float1Write=true
    Module.Index0Float2Write=true
    Module.Index0Float3Write=true
    Module.Index1Float0Write=true
    Module.Index1Float1Write=true
    Module.Index1Float2Write=true
    Module.Index1Float3Write=true
    Write Parameter Index 1=::NewEnumerator1
   )
   &ScaleColor(Scale Alpha=$System.Timeline001.Float, ScaleA=true, ScaleRGB=true)
   &ScaleSpriteSize(Scale Sprite Size Mode=::NewEnumerator0)
  ]
  rendererClasses [&NiagaraSpriteRendererProperties]
  renderers [
   {rendererClass=&NiagaraSpriteRendererProperties}
   {rendererClass=&NiagaraSpriteRendererProperties, rendererIndex=1}
  ]
 }
 Glows {
  emitterUpdateScript [
   &EmitterState(
    Inactive Response=::NewEnumerator1
    Life Cycle Mode=::NewEnumerator1
    Loop Behavior=::NewEnumerator1
    Loop Duration Mode=::NewEnumerator1
    UseLoopDelay=false
   )
   &SpawnBurst_Instantaneous(Spawn Count=9, Use Spawn Probability=false, UseLoopCountLimit=false)
  ]
  particleSpawnScript [
   &PartitionParticles(ID to Partition:NiagaraInt32=$Engine.ExecIndex, Number Of Partitions=9)
   &InitializeParticle(
    Color=&SelectLinearColorFromArray(
     Array Sampling Mode=::NewEnumerator1
     Direct Array Index=$Particles.PartitionParticles.Partition
     Color Selection Array=DataInterface() {propertyValues {ColorData [
        <rgba 1 1 1 0.1>
        <rgba 1 1 1 0.1>
        <rgba 1 1 1 0.1>
        <rgba 1 1 1 0.2>
        <rgba 1 1 1 0.05>
        <rgba 1 1 1 0.05>
        <rgba 1 1 1 0.05>
        <rgba 1 1 1 0.05>
        <rgba 1 1 1 0.2>
       ]}}
     Random Seed=?
    )
    Sprite Size=&SelectVector2DFromArray(
     Direct Array Index=$Particles.PartitionParticles.Partition
     Vector 2D Selection Array=DataInterface() {propertyValues {FloatData [<22 22>, <22 22>, <22 22>, <25 25>, <35 35>, <35 35>, <35 35>, <35 35>, <35 45>]}}
    )
    Sprite Size Mode=::NewEnumerator4
    UsePositionOffset=false
    Write Material Random=true
   )
  ]
  particleUpdateScript [
   &ParticleState(Kill Particles When Lifetime Has Elapsed=false, Let Infinitely Lived Particles Die When Emitter Deactivates=true)
   &SolveForcesAndVelocity(Clamp Velocity=false, Limit Acceleration=false)
   &SocketLocation(
    SamplePosOffset=true
    Sampled Position Offset=&SelectVectorFromArray(
     Direct Array Index=$Particles.PartitionParticles.Partition
     Vector Selection Array=DataInterface() {propertyValues {FloatData [<5 -7 0>, <8 -5 0>, <-8 5 0>, <-2 10 0>, <0 -1 -5.5572>, <-10 -1 -6.1558>, <0 1 5.5572>, <10 1 6.1558>, <2 8 0>]}}
    )
    Socket Index=&SelectIntFromArray(
     Direct Array Index=$Particles.PartitionParticles.Partition
     Int Selection Array=DataInterface() {propertyValues {IntData [10,11,12,10,6,6,8,8,1]}}
    )
    Socket Reader=$System.Socket Reader
    Socket Sampling Mode=::NewEnumerator3
    VelocitySpeedLimit=false
   )
   &CameraOffset(Camera Offset Amount=&FN_FrontEndFloatModifier(Gameplay=2, Front End=3))
   &ScaleColor(
    Scale Alpha=&Multiply_Float(
     A=&Multiply_Float(
      A=$System.Timeline001.Float
      B=&SelectFloatFromArray(Direct Array Index=$Engine.QualityLevel, Float Selection Array=DataInterface() {propertyValues {FloatData [0.5,0.5,1,1,1]}})
     )
     B=&FN_FrontEndFloatModifier(Gameplay=1, Front End=0.75)
    )
    ScaleA=true
    ScaleRGB=false
   )
  ]
  rendererClasses [&NiagaraSpriteRendererProperties]
  renderers [{rendererClass=&NiagaraSpriteRendererProperties}]
 }
 GlowsLow {
  emitterUpdateScript [
   &EmitterState(
    Inactive Response=::NewEnumerator1
    Life Cycle Mode=::NewEnumerator1
    Loop Behavior=::NewEnumerator1
    Loop Duration Mode=::NewEnumerator1
    UseLoopDelay=false
   )
   &SpawnBurst_Instantaneous(Spawn Count=12, Use Spawn Probability=false, UseLoopCountLimit=false)
  ]
  particleSpawnScript [
   &PartitionParticles(ID to Partition:NiagaraInt32=&ReturnExecIndex(), Number Of Partitions=12)
   &InitializeParticle(
    Color=<rgba 1 0 0.03 0.5>
    Sprite Size=&Multiply_Vector2D_ByFloat(
     Float=&FN_FrontEndFloatModifier(Gameplay=1, Front End=1.1)
     Vector2D=&SelectVector2DFromArray(
      Direct Array Index=$Particles.PartitionParticles.Partition
      Vector 2D Selection Array=DataInterface() {propertyValues {FloatData [<55 70>, <35 60>, <45 45>, <35 60>, <45 45>, <30 60>, <30 60>, <70 85>, <35 80>, <35 80>, <30 80>, <30 90>]}}
     )
    )
    Sprite Size Mode=::NewEnumerator4
    UsePositionOffset=false
    Write Material Random=true
   )
  ]
  particleUpdateScript [
   &ParticleState(Kill Particles When Lifetime Has Elapsed=false, Let Infinitely Lived Particles Die When Emitter Deactivates=true)
   &SolveForcesAndVelocity(Clamp Velocity=false, Limit Acceleration=false)
   &SocketLocation(
    Orientation Sampling=::NewEnumerator1
    SamplePosOffset=true
    Sampled Position Offset=&SelectVectorFromArray(
     Direct Array Index=$Particles.PartitionParticles.Partition
     Vector Selection Array=DataInterface() {propertyValues {FloatData [<5 0 0>, <16 0 0>, <-2 0 -2>, <-16 0 0>, <2 0 2>, <10 0 0>, <-10 0 0>, <0 0 0>, <-17 0 0>, <17 0 0>, <-15 0 0>, <15 0 0>]}}
    )
    Socket Index=&SelectIntFromArray(
     Direct Array Index=$Particles.PartitionParticles.Partition
     Int Selection Array=DataInterface() {propertyValues {IntData [13,2,2,4,4,3,5,0,6,8,7,9]}}
    )
    Socket Reader=$System.Socket Reader
    Socket Sampling Mode=::NewEnumerator3
    VelocitySpeedLimit=false
   )
   &CameraOffset(Camera Offset Amount=&FN_FrontEndFloatModifier(Gameplay=-2, Front End=-2))
   &ScaleColor(
    Scale Alpha=&Multiply_Float(
     A=&Multiply_Float(
      A=$System.Timeline001.Float
      B=&Multiply_Float(
       A=&MakeCustomFloatFromBool(
        Bool=$User.bisHeadHidden
        True Float=&SelectFloatFromArray(
         Direct Array Index=$Particles.PartitionParticles.Partition
         Float Selection Array=DataInterface() {propertyValues {FloatData [0,1,1,0,1,1,1,1,1,1,1,1]}}
        )
        False Float=1
       )
       B=&MakeCustomFloatFromBool(
        Bool=$User.bisHandHidden
        True Float=&SelectFloatFromArray(
         Direct Array Index=$Particles.PartitionParticles.Partition
         Float Selection Array=DataInterface() {propertyValues {FloatData [1,1,1,1,1,0,0,1,1,1,1,1]}}
        )
        False Float=1
       )
      )
     )
     B=&FN_FrontEndFloatModifier(
      Gameplay=1
      Front End=&SelectFloatFromArray(
       Direct Array Index=$Particles.PartitionParticles.Partition
       Float Selection Array=DataInterface() {propertyValues {FloatData [0.25,0.25,0.25,0.25,0.25,0.5,0.5,0.5,1,1,1,1]}}
      )
     )
    )
    ScaleA=true
    ScaleRGB=false
   )
   /SetVariables(Particles.SpriteAlignment=&RotateVectorByQuaternion(Vector To Rotate=<1 0 0>, Quaternion=$Output.SocketLocation.SampledRotation)) {isSetParameters=true}
  ]
  rendererClasses [&NiagaraSpriteRendererProperties]
  renderers [{rendererClass=&NiagaraSpriteRendererProperties}]
 }
}
summary {
 systemName=NS_Character_HonorBraceJoust_Idle
 emitters [
  {emitterName=GlowOverlay, rendererClasses [&NiagaraSpriteRendererProperties]}
  {emitterName=Em_Body_Glow, rendererClasses [&NiagaraMeshRendererProperties]}
  {emitterName=Glows, rendererClasses [&NiagaraSpriteRendererProperties]}
  {emitterName=GlowsLow, rendererClasses [&NiagaraSpriteRendererProperties]}
 ]
 userVariables [
  User.bIsFrontEnd {type=NiagaraBool}
  User.bIsFrontEndPreview {type=NiagaraBool}
  User.bisHandHidden {type=NiagaraBool}
  User.bisHeadHidden {type=NiagaraBool}
  User.Reactive {type=NiagaraBool}
 ]
}
