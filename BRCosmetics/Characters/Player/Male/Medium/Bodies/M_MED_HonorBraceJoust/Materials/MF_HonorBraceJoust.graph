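# Material function for the HonorBraceJoust player body cosmetic, layering a timed reactive emissive effect onto incoming material attributes.
# A timeline from Blueprint game time and duration, or a preview scalar, sequences decal glow, light scans, hand/feet Fresnel glow and fade-out.
# Packed texture channels and pre-skinned bounds position supply the masks.
# Reactivity_Toggle gates added emissive and darkened base color and specular; metallic, roughness and normal pass through.
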MaterialFunction /BRCosmetics/Characters/Player/Male/Medium/Bodies/M_MED_HonorBraceJoust/Materials/MF_HonorBraceJoust
graph {nodes [
  Frac(None=Multiply(A=Time()))
  FunctionInput() {bUsePreviewValueAsDefault=true, inputName=TechArtMask, sortPriority=1, previewValue=<1 1 1 1>}
  FunctionInput() {bUsePreviewValueAsDefault=true, inputName=Fresnel, inputType=FunctionInput_Scalar, sortPriority=3, previewValue=<1 1 1 1>}
  FunctionInput() {bUsePreviewValueAsDefault=true, inputName=PreSkinnedLocation, inputType=FunctionInput_Scalar, sortPriority=2, previewValue=<1 1 1 1>}
  n1 = GetMaterialAttributes(None=FunctionInput() {inputName=OUT_MA, inputType=FunctionInput_MaterialAttributes}) {attributeGetTypes [
    #1
    #2
    #3
    #4
    #5
    #6
   ]}
  n2 = SetMaterialAttributes!(
   MaterialAttributes=@n1.MaterialAttributes
   BaseColor=NamedRerouteUsage()
   Metallic=NamedRerouteUsage()
   Specular=NamedRerouteUsage()
   Roughness=NamedRerouteUsage()
   Normal=NamedRerouteUsage()
   EmissiveColor=NamedRerouteUsage()
  ) {attributeSetTypes [
    #1
    #2
    #3
    #4
    #5
    #6
   ]}
  FunctionOutput(None=@n2) {outputName=OUT_MA}
  n3 = NamedRerouteUsage()
  n4 = ScalarParameter() {defaultValue=1, group=HonorBraceJoust, parameterName=Reactivity_Toggle, sliderMax=1}
  NamedRerouteDeclaration(None=LinearInterpolate(A=NamedRerouteUsage(), B=Constant(), Alpha=Multiply(A=@n3, B=@n4))) {name=OUT_BaseColor, nodeColor=<rgba 0 0.8856 1 1>}
  NamedRerouteDeclaration(None=NamedRerouteDeclaration(None=@n1.Metallic) {name=IN_Metallic, nodeColor=<rgba 0.25 1 0 1>}) {name=OUT_Metallic, nodeColor=<rgba 0.3462 1 0.0006 1>}
  NamedRerouteDeclaration(None=@n1.EmissiveColor) {name=IN_EmissiveColor, nodeColor=<rgba 0 0.8 1 1>}
  n5 = TextureSample() {
   samplerSource=SSM_Wrap_WorldGroupSettings
   samplerType=SAMPLERTYPE_LinearColor
   texture=/BRCosmetics/Characters/Player/Male/Medium/Bodies/M_MED_HonorBraceJoust/Textures/T_HonorBraceJoust_Body_Emissive_FX
   channelNames {a="", b=ID, g=Glow, r=Mask}
  }
  NamedRerouteDeclaration(None=@n5.ID) {name=ID, nodeColor=<rgba 0.9219 0 1 1>}
  n6 = LocalPosition() {localOrigin=InstancePreSkinning}
  NamedRerouteDeclaration(None=ShaderStageSwitch(PixelShader=VertexInterpolator(VS=@n6.XYZ).PS, VertexShader=@n6.XYZ)) {name=PreSkinnedLoc, nodeColor=<rgba 0.7001 0 1 1>}
  n7 = NamedRerouteUsage()
  n8 = Step(Y=@n7) {constX=0.4, constY=0.4}
  n9 = NamedRerouteUsage()
  n10 = Step(Y=@n7) {constX=0.2, constY=0.4}
  NamedRerouteDeclaration(None=Subtract(A=Multiply(A=@n8, B=@n9), B=@n10)) {name=Shock_Decal, nodeColor=<rgba 1 0 0.1172 1>}
  n11 = Multiply(A=Step(Y=@n7) {constX=0.6, constY=0.4}, B=@n9)
  NamedRerouteDeclaration(None=Subtract(A=@n11, B=@n8)) {name=Pants_Decal, nodeColor=<rgba 1 0 0.1172 1>}
  n12 = Multiply(A=Step(Y=@n7) {constX=0.85, constY=0.4}, B=@n9)
  NamedRerouteDeclaration(None=Subtract(A=@n12, B=@n11)) {name=Shoulder_Decal, nodeColor=<rgba 1 0 0.1172 1>}
  n13 = Step(Y=@n7) {constX=1.2, constY=0.4}
  NamedRerouteDeclaration(None=Subtract(A=Multiply(A=@n13, B=@n9), B=@n12)) {name=TextLogo_Decal, nodeColor=<rgba 1 0 0.1172 1>}
  n14 = NamedRerouteUsage()
  NamedRerouteDeclaration(None=LinearInterpolate(A=NamedRerouteUsage(), B=Constant(), Alpha=Multiply(A=@n14, B=@n4)) {constB=0}) {name=OUT_Specular, nodeColor=<rgba 0.25 1 0 1>}
  n15 = NamedRerouteDeclaration(None=@n5.Glow) {name=GlowMask, nodeColor=<rgba 0.5194 1 0 1>}
  n16 = NamedRerouteDeclaration(None=@n5.Mask) {name=Mask, nodeColor=<rgba 0 1 0.3203 1>}
  NamedRerouteDeclaration(None=Saturate(None=Subtract(A=@n15, B=@n16))) {name=EdgeGlow, nodeColor=<rgba 0.7344 1 0 1>}
  NamedRerouteDeclaration(None=Multiply(A=@n10, B=@n9)) {name=Shose_Decal, nodeColor=<rgba 1 0 0.1172 1>}
  n17 = Max(
   A=Multiply(A=ScalarParameter() {desc=0-1 input. 0.75 at Perfect Stage, group=HonorBraceJoust, parameterName=ReactivityPreview, sliderMax=1}) {constB=2}
   B=NamedRerouteUsage()
  )
  n18 = Constant4Vector() {constant=<rgba 0 1 0 6>}
  n19 = Modulo!(
   A=NamedRerouteDeclaration(None=MaterialFunctionCall(Input=@n17, Input Low=@n18.R, Input High=@n18.R, Target Low=@n18.R, Target High=@n18.R) {materialFunction=&RemapValueRange}.Result) {name=FixedTimeline, nodeColor=<rgba 1 0.4922 0 1>}
   B=Constant() {r=6}
  )
  n20 = Constant4Vector() {constant=<rgba 1 2 0 1>}
  n21 = MaterialFunctionCall(Input=@n19, Input Low=@n20.R, Input High=@n20.R, Target Low=@n20.R, Target High=@n20.R) {materialFunction=&RemapValueRange}
  n22 = Power(Base=NamedRerouteUsage())
  n23 = NamedRerouteUsage()
  n24 = Multiply(
   A=Multiply(A=Saturate(None=@n21.Result), B=Multiply(A=@n22) {constB=0.5})
   B=LinearInterpolate(A=VectorParameter() {group=HonorBraceJoust, parameterName=Decal_GlowingColor, defaultValue=<rgba 1 0 0.05 1>}.RGB, Alpha=@n23)
  )
  n25 = LinearInterpolate(A=ScalarParameter() {defaultValue=5, group=HonorBraceJoust, parameterName=GlowingColor_Intensity}, Alpha=@n23)
  n26 = Multiply(A=@n19, B=NamedRerouteUsage()) {constB=0.75}
  n27 = NamedRerouteUsage()
  n28 = ComponentMask(None=@n27) {r=true}
  n29 = ComponentMask(None=@n27) {b=true}
  n30 = NamedRerouteUsage()
  n31 = Subtract(A=@n29, B=@n30)
  n32 = Add(A=@n29, B=Divide(A=@n30))
  n33 = Step(Y=ComponentMask(None=@n27) {g=true}) {constX=0.3, constY=0.3, desc=FrontBackMask}
  n34 = Frac(None=Add(A=Multiply(A=@n28) {constB=3}, B=LinearInterpolate(A=@n31, B=@n32, Alpha=@n33) {desc=PanningDirection}))
  n35 = Abs(None=Subtract(A=Multiply(A=@n34) {constB=2}))
  n36 = Saturate(None=Divide(B=OneMinus(None=@n35)) {constA=0.05}) {desc=HorizontalLightScan}
  n37 = Multiply(A=Max(A=Multiply(A=@n26) {constB=0.75}, B=Multiply(A=@n36, B=@n19) {constB=0.5}), B=NamedRerouteUsage())
  n38 = Frac(None=Multiply(A=@n32) {constB=2})
  n39 = Abs(None=Subtract(A=Multiply(A=@n38) {constB=2}))
  n40 = Saturate(None=Divide(B=OneMinus(None=@n39)) {constA=0.05}) {desc=VerticalLightScan}
  n41 = Constant4Vector() {constant=<rgba 1 1.5 0 1>}
  n42 = Saturate(None=MaterialFunctionCall(Input=@n19, Input Low=@n41.R, Input High=@n41.R, Target Low=@n41.R, Target High=@n41.R) {materialFunction=&RemapValueRange}.Result)
  n43 = Multiply(A=ComponentMask(None=NamedRerouteUsage()) {b=true}) {constB=2.4}
  n44 = ConstantBiasScale(None=Sine(None=Add(A=@n43) {constB=-0.25}))
  n45 = Constant4Vector() {constant=<rgba 0.5 1 0 1>}
  n46 = OneMinus(None=Saturate(None=MaterialFunctionCall(Input=@n19, Input Low=@n45.R, Input High=@n45.R, Target Low=@n45.R, Target High=@n45.R) {materialFunction=&RemapValueRange}.Result))
  n47 = Divide(A=Floor(None=Multiply(A=@n46) {constB=10})) {constB=6}
  n48 = Constant4Vector() {constant=<rgba 0 0.1 0 1>}
  n49 = Multiply(
   A=Saturate(None=MaterialFunctionCall!(Mask In=@n44, Amount=@n47, Range=@n48.G, Start Offset=@n48.G, Sharpness=@n48.G) {materialFunction=/Game/Global/MaterialFunctions/MF_MaskShifting}.Result)
   B=NamedRerouteUsage()
  ) {constB=0.75}
  n50 = Multiply(A=Max(A=Multiply(A=@n40, B=@n42), B=Multiply(A=@n49) {constB=0.75}), B=NamedRerouteUsage())
  n51 = NamedRerouteUsage()
  n52 = Constant4Vector() {constant=<rgba 4 5 0 1>}
  n53 = Saturate(None=MaterialFunctionCall(Input=@n19, Input Low=@n52.R, Input High=@n52.R, Target Low=@n52.R, Target High=@n52.R) {materialFunction=&RemapValueRange}.Result)
  n54 = Constant4Vector() {constant=<rgba 0 0.5 0 1>}
  n55 = Min(
   A=LinearInterpolate(A=Multiply(A=@n24, B=@n25) {constB=10}, B=Add(A=@n37, B=@n50), Alpha=NamedRerouteUsage()) {desc=PutGlow At Emissition Back}
   B=Saturate(None=MaterialFunctionCall(Mask In=@n51, Amount=@n53, Range=@n54.G, Start Offset=@n54.G, Sharpness=@n54.G) {materialFunction=&MF_MaskShifting}.Result)
  ) {desc=FadeOut Order}
  n56 = Add(A=NamedRerouteUsage(), B=@n55)
  n57 = Add(A=NamedRerouteUsage(), B=NamedRerouteUsage())
  NamedRerouteDeclaration(None=Multiply(A=Multiply(A=@n56, B=@n57), B=ScalarParameter() {defaultValue=1.5, group=HonorBraceJoust, parameterName=EmissiveFX_Intensity}) {constB=2}) {name=ReactiveEmission, nodeColor=<rgba 0 1 0.9766 1>}
  NamedRerouteDeclaration(None=OneMinus(None=Fresnel() {exponent=1.5})) {name=Fresnel_Decals, nodeColor=<rgba 1 0.3047 0 1>}
  n58 = NamedRerouteUsage()
  n59 = ComponentMask(None=@n58) {r=true}
  n60 = Abs(None=Subtract(A=Multiply(A=@n59) {constB=2}))
  n61 = ComponentMask(None=@n58) {b=true}
  n62 = Constant4Vector() {constant=<rgba 0 0.5 0.25 1>}
  n63 = Saturate(None=MaterialFunctionCall(Mask In=Subtract(A=@n60, B=@n61), Amount=@n62.R, Range=@n62.R, Start Offset=@n62.R, Sharpness=@n62.R) {materialFunction=&MF_MaskShifting}.Result) {desc="Hand&&FeetMask"}
  n64 = NamedRerouteUsage()
  NamedRerouteDeclaration(None=Multiply(A=Multiply(A=@n63, B=@n64), B=NamedRerouteUsage())) {name=HandFeet_Mask, nodeColor=<rgba 1 0.4219 0 1>}
  n65 = Desaturation(None=@n55)
  NamedRerouteDeclaration(None=Saturate(None=Multiply(A=@n65) {constB=2})) {name=FX_Mask, nodeColor=<rgba 0 1 0.8828 1>}
  NamedRerouteDeclaration(None=@n12) {name=Lightning_Decal, nodeColor=<rgba 1 0 0.1172 1>}
  NamedRerouteDeclaration(None=Step(X=@n7) {constX=0.9, constY=0.9}) {name=TextLogo_Mask, nodeColor=<rgba 0.1719 0 1 1>}
  NamedRerouteDeclaration(None=NamedRerouteDeclaration(None=@n1.Roughness) {name=IN_Roughness, nodeColor=<rgba 0.25 1 0 1>}) {name=OUT_Roughness, nodeColor=<rgba 0.25 1 0 1>}
  n66 = LinearInterpolate(Alpha=NamedRerouteUsage()) {constA=2.5}
  n67 = Multiply(A=Power(Base=Fresnel(ExponentIn=@n66) {exponent=2}), B=NamedRerouteUsage())
  n68 = LinearInterpolate(
   A=VectorParameter() {group=HonorBraceJoust, parameterName=Outline_GlowingColor, defaultValue=<rgba 1 0 0.05 1>}.RGB
   Alpha=NamedRerouteUsage()
  ) {constB=0.6}
  NamedRerouteDeclaration(None=Multiply(A=Multiply(A=@n67, B=@n68), B=ScalarParameter() {defaultValue=10, group=HonorBraceJoust, parameterName=OutlineGlow_Intensity}) {constB=10}) {name=HandFeetGlow, nodeColor=<rgba 0.1484 1 0 1>}
  NamedRerouteDeclaration(None=OneMinus(None=@n53)) {name=Timeline_HandFeet_FadeOut, nodeColor=<rgba 1 0 0.4687 1>}
  n69 = TextureCoordinate()
  n70 = TextureSample(UVs=Panner(Coordinate=Multiply(A=@n69) {constB=4}) {speedY=0.2}) {samplerSource=SSM_Wrap_WorldGroupSettings, samplerType=SAMPLERTYPE_Grayscale, texture=&LowResBlurredNoise}
  NamedRerouteDeclaration(None=Power(Base=@n70.R)) {name=ColorNoise, nodeColor=<rgba 1 0 0.7031 1>}
  n71 = Constant4Vector() {constant=<rgba 0 4 0 1>}
  NamedRerouteDeclaration(None=Saturate(None=MaterialFunctionCall(Input=@n19, Input Low=@n71.R, Input High=@n71.R, Target Low=@n71.R, Target High=@n71.R) {materialFunction=&RemapValueRange}.Result)) {name=Timeline_HandFeet_FadeIN, nodeColor=<rgba 0 1 0.5078 1>}
  n72 = TransformPosition(None=NamedRerouteUsage()) {transformType=TRANSFORMPOSSOURCE_Instance}
  n73 = PreSkinnedLocalBounds()
  NamedRerouteDeclaration(None=Divide(A=Subtract(A=@n72, B=@n73.Min), B=@n73.Extents)) {name=Bounding, nodeColor=<rgba 0 0.5312 1 1>}
  NamedRerouteDeclaration(None=NamedRerouteDeclaration(None=@n1.Normal) {name=IN_Normal, nodeColor=<rgba 0.25 0.25 1 1>}) {name=OUT_Normal, nodeColor=<rgba 0.25 0.25 1 1>}
  n74 = Subtract(A=Time(), B=ScalarParameter() {defaultValue=-99999, group=HonorBraceJoust, parameterName=GameTimeFromBP})
  n75 = Add(A=ScalarParameter() {defaultValue=3, group=HonorBraceJoust, parameterName=AnimDuration}) {constB=0}
  NamedRerouteDeclaration(None=Saturate(None=Divide(A=@n74, B=@n75))) {name=TimeNative, nodeColor=<rgba 0.2891 1 0 1>}
  n76 = NamedRerouteUsage()
  NamedRerouteDeclaration(None=Add(A=NamedRerouteUsage(), B=Multiply(A=@n76, B=@n4))) {name=OUT_EmissiveColor, nodeColor=<rgba 0 0.8 1 1>}
  NamedRerouteDeclaration(None=@n1.BaseColor) {name=IN_BaseColor, nodeColor=<rgba 0 0.8 1 1>}
  NamedRerouteDeclaration(None=@n1.Specular) {name=IN_Specular, nodeColor=<rgba 0.25 1 0 1>}
 ]}
