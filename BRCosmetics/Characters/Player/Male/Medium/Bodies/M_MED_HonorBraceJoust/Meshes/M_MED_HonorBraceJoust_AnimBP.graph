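# Anim Blueprint for the HonorBraceJoust male medium body cosmetic, adding secondary motion to shorts and driving material-based cloth simulation.
# AnimGraph copies the parent mesh pose, runs a Control Rig, then chains four AnimDynamics nodes on front/back shorts bones.
# Initialization and customization-loaded events cache the mesh and dynamic base/overlay materials.
# Each update feeds calf-driven WPO regions and sets wind and turbulence, raised during skydiving or parachuting, zeroed otherwise.
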AnimBlueprint /BRCosmetics/Characters/Player/Male/Medium/Bodies/M_MED_HonorBraceJoust/Meshes/M_MED_HonorBraceJoust_AnimBP
target_skeleton=&/M_MED_HonorBraceJoust_Skeleton
graphs [
 AnimGraph {nodes [
   n1 = Control Rig(Source=Copy Pose From Mesh() {bUseAttachedParent=true}) {controlRigAssetReference {blueprintRigClass=&/M_MED_HonorBraceJoust_CtrlRig.M_MED_HonorBraceJoust_CtrlRig_C}}
   n2 = Anim Dynamics - Bone: dyn_shorts_fr_l!(ComponentPose=Local To Component(LocalPose=@n1), Alpha=1, GravityOverride:Vector="0,0,-980", ExternalForce="0,0,0") -> (Pose) {
    angularDampingOverride=1
    angularSpringConstant=1500
    bOverrideAngularDamping=true
    bUseGravityOverride=true
    gravityScale=0.35
    linearDampingOverride=0.9999
    simulationSpace=RootRelative
    alphaBoolBlend {blendOption=Linear}
    boundBone {boneName=dyn_shorts_fr_l}
    componentAppliedLinearAccClamp=<2000 2000 2000>
    componentLinearAccScale=<0.2 0.2 0.2>
    componentLinearVelScale=<0.2 0.2 0.2>
    physicsBodyDefinitions [{
      boundBone {boneName=dyn_shorts_fr_l}
      boxExtents=<10 20 20>
      constraintSetup {angularLimitsMax=<0 10 2>, angularLimitsMin=<0 -20 -2>, angularTarget=<1 0 0>}
      localJointOffset=<-2 0 0>
     }]
    relativeSpaceBone {boneName=dyn_shorts_l_base}
    simSpaceSettings {maxAngularAcceleration=200, maxAngularVelocity=15}
   }
   n3 = Anim Dynamics - Bone: dyn_shorts_bk_l!(ComponentPose=@n2, Alpha=1, GravityOverride:Vector="0,0,-980", ExternalForce="0,0,0") -> (Pose) {
    angularDampingOverride=1
    angularSpringConstant=1500
    bOverrideAngularDamping=true
    bUseGravityOverride=true
    gravityScale=0.35
    linearDampingOverride=0.9999
    simulationSpace=RootRelative
    alphaBoolBlend {blendOption=Linear}
    boundBone {boneName=dyn_shorts_bk_l}
    componentAppliedLinearAccClamp=<2000 2000 2000>
    componentLinearAccScale=<0.2 0.2 0.2>
    componentLinearVelScale=<0.2 0.2 0.2>
    physicsBodyDefinitions [{
      boundBone {boneName=dyn_shorts_bk_l}
      boxExtents=<10 20 20>
      constraintSetup {angularLimitsMax=<0 10 2>, angularLimitsMin=<0 -20 -2>, angularTarget=<1 0 0>}
      localJointOffset=<-2 0 0>
     }]
    relativeSpaceBone {boneName=dyn_shorts_l_base}
    simSpaceSettings {maxAngularAcceleration=200, maxAngularVelocity=15}
   }
   n4 = Anim Dynamics - Bone: dyn_shorts_fr_r!(ComponentPose=@n3, Alpha=1, GravityOverride:Vector="0,0,-980", ExternalForce="0,0,0") -> (Pose) {
    angularDampingOverride=1
    angularSpringConstant=1500
    bAngularSpring=true
    bOverrideAngularDamping=true
    bUseGravityOverride=true
    gravityScale=0.35
    linearDampingOverride=0.9999
    simulationSpace=RootRelative
    alphaBoolBlend {blendOption=Linear}
    boundBone {boneName=dyn_shorts_fr_r}
    componentAppliedLinearAccClamp=<2000 2000 2000>
    componentLinearAccScale=<0.2 0.2 0.2>
    componentLinearVelScale=<0.2 0.2 0.2>
    physicsBodyDefinitions [{
      boundBone {boneName=dyn_shorts_fr_r}
      boxExtents=<10 20 20>
      constraintSetup {angularLimitsMax=<0 10 2>, angularLimitsMin=<0 -20 -2>, angularTarget=<1 0 0>}
      localJointOffset=<2 0 0>
     }]
    relativeSpaceBone {boneName=dyn_shorts_r_base}
    simSpaceSettings {maxAngularAcceleration=200, maxAngularVelocity=15}
   }
   n5 = Anim Dynamics - Bone: dyn_shorts_bk_r!(ComponentPose=@n4, Alpha=1, GravityOverride:Vector="0,0,-980", ExternalForce="0,0,0") -> (Pose) {
    angularDampingOverride=1
    angularSpringConstant=1500
    bAngularSpring=true
    bOverrideAngularDamping=true
    bUseGravityOverride=true
    gravityScale=0.35
    linearDampingOverride=0.9999
    simulationSpace=RootRelative
    alphaBoolBlend {blendOption=Linear}
    boundBone {boneName=dyn_shorts_bk_r}
    componentAppliedLinearAccClamp=<2000 2000 2000>
    componentLinearAccScale=<0.2 0.2 0.2>
    componentLinearVelScale=<0.2 0.2 0.2>
    physicsBodyDefinitions [{
      boundBone {boneName=dyn_shorts_bk_r}
      boxExtents=<10 20 20>
      constraintSetup {angularLimitsMax=<0 10 2>, angularLimitsMin=<0 -20 -2>, angularTarget=<1 0 0>}
      localJointOffset=<2 0 0>
     }]
    relativeSpaceBone {boneName=dyn_shorts_r_base}
    simSpaceSettings {maxAngularAcceleration=200, maxAngularVelocity=15}
   }
   Output Pose(Result=Component To Local(ComponentPose=@n5)) {layerGroup=None}
  ]}
 EventGraph {nodes [
   n1 = Event BlueprintUpdateAnimation()
   > n2 = Sequence()
   > n3 = Gravity Override State Function DEPRECATED(CharacterPartAnimInstance=Self-Reference(), GameTimeInSeconds=GetGameTimeInSeconds(), GravityOverrideParams=Get Gravity Override Params())
   > Set Gravity Override(Gravity Override=@n3)
   n4 = Sequence(execute=@n2.then_1)
   > n5 = Branch(Condition=OR Boolean!(A=Get bIsSkydiveFloating(), B=Get bIsSkydiveDiveMode(), C:bool=Get bIsParachuteOpened()) -> (ReturnValue))
   > Set Velocity Interp Speed(Velocity Interp Speed:double=30) -> (Output_Get:double)
   > Set WPO Wind(WPO Wind:double=1) -> (Output_Get:double)
   > Set Turbulence Noise(Turbulence Noise:float=75) -> (Output_Get:float)
   Set Velocity Interp Speed(execute=@n5.else, Velocity Interp Speed=0)
   > Set WPO Wind(WPO Wind=0)
   > Set Turbulence Noise(Turbulence Noise=0)
   n6 = Get Skeletal Mesh() -> (Skeletal Mesh:Skeletal Mesh Component Object Reference)
   n7 = Get Overlay Material Instance Dynamic() -> (Overlay Material Instance Dynamic:Material Instance Dynamic Object Reference)
   n8 = Get Material Instance Dynamic() -> (Material Instance Dynamic:Material Instance Dynamic Object Reference)
   Set WPO Globals Function(
    execute=@n4.then_1
    Material=@n8
    Overlay Material=@n7
    Skeletal Mesh=@n6
    "Local Noise Transform  Location"="Get Local Noise Transform  Location"() -> ("Local Noise Transform  Location":Vector)
    Local Noise Offset=Get Local Noise Offset()
   )
   > Set WPO Regional Function!(
    execute
    Material:Material Instance Dynamic Object Reference=@n8
    Overlay Material:Material Instance Dynamic Object Reference=@n7
    Skeletal Mesh:Skeletal Mesh Component Object Reference=@n6
    Velocity Transform:Name=calf_l
    Velocity Interp Speed:double=Get Velocity Interp Speed()
    Collision Transform:Name=calf_l
    Collision Transform Axis:EAxis Enum
    Invert Collision Transform Axis:bool=false
    Delta Time X:double=@n1.DeltaTimeX
    Previous Velocity Transform Location:"Vector (by ref)"=Get R2 Velocity Transform Location()
    Previous Velocity Transform Velocity:"Vector (by ref)"=Get R2 Velocity Transform Velocity()
    Region:int=1
   ) -> (then)
   > Set WPO Regional Function(
    Material=@n8
    Overlay Material=@n7
    Skeletal Mesh=@n6
    Velocity Transform=calf_r
    Velocity Interp Speed=Get Velocity Interp Speed()
    Collision Transform=calf_r
    Invert Collision Transform Axis=true
    Delta Time X=@n1.DeltaTimeX
    Previous Velocity Transform Location=Get L Previous Velocity Transform Location() -> (L Previous Velocity Transform Location:Vector)
    Previous Velocity Transform Velocity=Get L Previous Velocity Transform Velocity() -> (L Previous Velocity Transform Velocity:Vector)
    Region=2
   )
   > SetScalarParameterValue(self=[@n8, @n7], ParameterName=ClothSim_WindStrength, Value=Get WPO Wind() -> (WPO Wind:double))
   > SetScalarParameterValue(self=[@n8, @n7], ParameterName=ClothSim_TurbulenceNoiseTileDivisor, Value=Get Turbulence Noise() -> (Turbulence Noise:float))
   n9 = Event OnCharacterCustomizationLoaded()
   n10 = Sequence(execute=[@n9.then, @n16.then])
   n11 = TryGetPawnOwner()
   > Is Valid(InputObject=@n11)
   > n12 = Cast To FortPlayerPawnAthena(Object=@n11)
   > Set As Fort Player Pawn Athena(As Fort Player Pawn Athena:Fort Player Pawn Athena Object Reference=@n12) -> (Output_Get:Fort Player Pawn Athena Object Reference)
   n13 = Set Skeletal Mesh(execute=@n10.then_1, Skeletal Mesh:Skeletal Mesh Component Object Reference=GetPartSkeletalMeshComponent()) -> (Output_Get:Skeletal Mesh Component Object Reference)
   > n14 = Get Dynamic Material Instance(SkeletalMeshComponent=@n13)
   > Set Material Instance Dynamic(Material Instance Dynamic:Material Instance Dynamic Object Reference=@n14) -> (Output_Get:Material Instance Dynamic Object Reference)
   > n15 = GetOrCreateDynamicOverlayMaterial(MeshComponent=@n13, MaterialName=None)
   > Set Overlay Material Instance Dynamic(Overlay Material Instance Dynamic:Material Instance Dynamic Object Reference=@n15) -> (Output_Get:Material Instance Dynamic Object Reference)
   n16 = Event BlueprintInitializeAnimation()
  ]}
]
variables [
 As Fort Player Pawn Athena {type=FortPlayerPawnAthena}
 Gravity Override {type=Vector}
 Gravity Override Params {type=GravityOverrideParamsStruct}
 L Previous Velocity Transform Location {type=Vector}
 L Previous Velocity Transform Velocity {type=Vector}
 Local Noise Offset {type=Vector}
 "Local Noise Transform  Location" {type=Vector}
 Material Instance Dynamic {type=MaterialInstanceDynamic}
 Overlay Material Instance Dynamic {type=MaterialInstanceDynamic}
 R2 Velocity Transform Location {type=Vector}
 R2 Velocity Transform Velocity {type=Vector}
 Skeletal Mesh {type=SkeletalMeshComponent}
 Turbulence Noise {type=number}
 Velocity Interp Speed {type=number}
 WPO Wind {type=number}
]
