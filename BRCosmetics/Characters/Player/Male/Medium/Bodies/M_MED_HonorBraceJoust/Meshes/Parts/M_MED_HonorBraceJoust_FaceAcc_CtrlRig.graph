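# Control Rig for the HonorBraceJoust medium male face accessory that adds procedural flutter to its dyn_hair bones while skydiving.
# When the skydiving variable is true, a loop over an array of hair bones modifies each rotation through ModifyTransforms.
# The RandomScaledRotation function builds that rotation from per-axis noise, with speed set by "speed scale" and symmetric ranges set by seeded, timed random values.
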ControlRigBlueprint /BRCosmetics/Characters/Player/Male/Medium/Bodies/M_MED_HonorBraceJoust/Meshes/Parts/M_MED_HonorBraceJoust_FaceAcc_CtrlRig
graphs [
 New Function_ContainedGraph {nodes [
   n1 = RigVMFunctionEntryNode()
   Random = /Script/RigVM.RigVMFunction_RandomFloat(Duration=10, Maximum=15, Minimum=0, Seed=200) {is=struct RigVMUnitNode, template="Random::Execute(in Seed,in Minimum,in Maximum,in Duration,out Result)"}
   Multiply = &RigVMFunction_MathDoubleMul(B=-1, A=@Random.Result)
   Noise = /Script/RigVM.RigVMFunction_NoiseFloat(Frequency=0.5, Value=0.5, Speed=@n1.Scale, Minimum=@Multiply.Result, Maximum=@Random.Result) {is=struct RigVMUnitNode, template="Noise::Execute(in Value,in Speed,in Frequency,in Minimum,in Maximum,out Result)"}
   n2 = &RigVMFunction_RandomFloat(Duration=10, Maximum=15, Minimum=0, Seed=600) {template="Random::Execute(in Seed,in Minimum,in Maximum,in Duration,out Result)"}
   n3 = &RigVMFunction_RandomFloat(Duration=10, Maximum=10, Minimum=0, Seed=150) {template="Random::Execute(in Seed,in Minimum,in Maximum,in Duration,out Result)"}
   RigVMFunction_MathQuaternionFromEuler = &RigVMFunction_MathQuaternionFromEuler(
    Euler=(X=0,Y=0,Z=0)
    Euler.X=@Noise.Result
    Euler.Y=&RigVMFunction_NoiseFloat(Frequency=0.5, Value=0.3, Speed=@n1.Scale, Minimum=&RigVMFunction_MathDoubleMul(B=-1, A=@n2.Result).Result, Maximum=@n2.Result) {template="Noise::Execute(in Value,in Speed,in Frequency,in Minimum,in Maximum,out Result)"}.Result
    Euler.Z=&RigVMFunction_NoiseFloat(Frequency=0.5, Value=0.5, Speed=@n1.Scale, Minimum=&RigVMFunction_MathDoubleMul(B=-1, A=@n3.Result).Result, Maximum=@n3.Result) {template="Noise::Execute(in Value,in Speed,in Frequency,in Minimum,in Maximum,out Result)"}.Result
   )
   > RigVMFunctionReturnNode(Result=@RigVMFunction_MathQuaternionFromEuler.Result)
  ]}
 RigVMFunctionLibrary {nodes [RandomScaledRotation = RigVMCollapseNode()]}
 RigVMModel {nodes [
   RigUnit_BeginExecution = &RigUnit_BeginExecution()
   RigVMFunction_ControlFlowBranch = &RigVMFunction_ControlFlowBranch(ExecuteContext=@RigUnit_BeginExecution.ExecutePin, Condition=RigVMVariableNode() {variable=skydiving}.Value)
   RerouteNode = RigVMDispatchNode(Values="((Type=Bone,Name=\"dyn_hair_a\"),((),Type=Bone,Name=\"dyn_hair_b\"),((),Type=Bone,Name=\"dyn_hair_c\"),((),Type=Bone,Name=\"dyn_hair_d\"),((),Type=Bone,Name=\"dyn_hair_e\"),((),Type=Bone,Name=\"dyn_hair_f\"),((),Type=Bone,Name=\"dyn_hair_g\"),((),Type=Bone,Name=\"dyn_hai ...(truncated)") {template="DISPATCH_RigVMDispatch_ArrayMake(in Values,out Array)"}
   n1 = /Script/RigVM.RigVMFunction_ForLoopCount(
    Count=RigVMDispatchNode(Array=@RerouteNode.Array) {template="DISPATCH_RigVMDispatch_ArrayGetNum(in Array,out Num)"}.Num
    ExecutePin=@RigVMFunction_ControlFlowBranch.True
   ) {is=struct RigVMUnitNode}
   RandomScaledRotation = RigVMFunctionReferenceNode(Scale=RigVMVariableNode() {variable=speed scale}.Value)
   > &RigUnit_ModifyTransforms(
    ItemToModify=((Item=(Type=Bone),Transform=(Rotation=(X=0,Y=0,Z=0,W=1),Translation=(X=0,Y=0,Z=0),Scale3D=(X=1,Y=1,Z=1))))
    Weight=1
    ItemToModify.0.Item=RigVMDispatchNode(Index=@n1.Index, Array=@RerouteNode.Array) {template="DISPATCH_RigVMDispatch_ArrayGetAtIndex(in Array,in Index,out Element)"}.Element
    ItemToModify.0.Transform.Rotation=@RandomScaledRotation.Result
   )
  ]}
]
hierarchy [
 root {loc=<0 0 0>, rot=<0 0 0 1>}
 pelvis {parent=root, loc=<0 -0.7071 93.1191>, rot=<0.0127 0.707 -0.0127 -0.707>}
 spine_01 {parent=pelvis, loc=<6.7494 -0.6388 0>, rot=<0 0 0.034 -0.9994>}
 spine_02 {parent=spine_01, loc=<7.2518 -0.035 0>, rot=<0 0 0.0238 -0.9997>}
 spine_03 {parent=spine_02, loc=<6.0104 0 0>, rot=<0 0 0.0149 -0.9999>}
 spine_04 {parent=spine_03, loc=<8.1024 0 0>, rot=<0 0 -0.0132 -0.9999>}
 spine_05 {parent=spine_04, loc=<10.0146 0 0>, rot=<0 0 -0.0015 -1>}
 neck_01 {parent=spine_05, loc=<11.3138 0.1529 0>, rot=<0 0 -0.2003 -0.9797>}
 neck_02 {parent=neck_01, loc=<4.5627 0 0>, rot=<0 0 0.0504 -0.9987>}
 head {parent=neck_02, loc=<5.3914 0 0>, rot=<0 0 0.115 -0.9934>}
 dyn_hair {parent=head, loc=<11.2734 0 0>, rot=<0 0 0 1>}
 dyn_hair_a {parent=dyn_hair, loc=<-0.0669 6.2347 -2.1599>, rot=<0.0603 -0.3522 0.1184 -0.9265>}
 dyn_hair_b {parent=dyn_hair, loc=<4.5519 -2.4777 -4.0602>, rot=<0.222 -0.3566 0.2839 -0.862>}
 dyn_hair_c {parent=dyn_hair, loc=<3.6654 3.7508 1.8086>, rot=<-0.1628 0.2141 0.2516 -0.9297>}
 dyn_hair_d {parent=dyn_hair, loc=<2.5604 6.7086 2.086>, rot=<0.0358 0.4227 -0.2157 -0.8795>}
 dyn_hair_e {parent=dyn_hair, loc=<0.2641 3.8297 6.5803>, rot=<-0.4573 -0.6975 -0.364 0.4145>}
 dyn_hair_f {parent=dyn_hair, loc=<-0.4797 -1.5873 6.1272>, rot=<0.0851 -0.8769 -0.182 0.4366>}
 dyn_hair_g {parent=dyn_hair, loc=<-3.2167 -4.126 5.9588>, rot=<0.1189 -0.873 -0.1987 0.4293>}
 dyn_hair_h {parent=dyn_hair, loc=<-3.5862 -6.9422 6.1468>, rot=<0.3579 -0.7623 -0.2463 0.4797>}
 dyn_hair_i {parent=dyn_hair, loc=<-2.9747 -10.0947 2.3135>, rot=<-0.2836 0.5325 0.6481 -0.4647>}
 dyn_hair_j {parent=dyn_hair, loc=<-2.9251 -9.0445 -2.3979>, rot=<-0.3008 0.5861 0.6447 -0.3877>}
 dyn_hair_k {parent=dyn_hair, loc=<-2.2642 -7.5138 -5.9047>, rot=<-0.4569 0.4923 0.6606 -0.3355>}
 dyn_hair_l {parent=dyn_hair, loc=<-2.6206 -3.3116 -7.721>, rot=<0.4399 -0.7207 -0.5357 0.0075>}
 dyn_hair_m {parent=dyn_hair, loc=<-0.7618 2.2765 -6.9241>, rot=<0.3182 -0.6701 -0.6203 -0.2548>}
 dyn_hair_n {parent=dyn_hair, loc=<2.472 -6.6867 3.5214>, rot=<-0.2935 0.6151 0.191 -0.7064>}
 dyn_hair_o {parent=dyn_hair, loc=<0.3757 -10.2226 2.0134>, rot=<-0.4482 0.5135 0.374 -0.6289>}
 dyn_hair_p {parent=dyn_hair, loc=<-0.0091 -8.9651 -2.0698>, rot=<-0.4346 0.5091 0.6643 -0.3327>}
 dyn_hair_q {parent=dyn_hair, loc=<3.8637 -6.2664 -0.435>, rot=<-0.5669 0.3635 0.5722 -0.468>}
 dyn_hair_r {parent=dyn_hair, loc=<1.6233 -8.8624 -2.9745>, rot=<0.6421 -0.3622 -0.5606 0.3773>}
 dyn_hair_s {parent=dyn_hair, loc=<2.5006 -6.047 -4.21>, rot=<0.7551 -0.0472 -0.6297 0.1763>}
 dyn_hair_t {parent=dyn_hair, loc=<3.5341 -0.0858 5.3511>, rot=<0.1588 0.466 0.2433 -0.8357>}
 dyn_hair_y {parent=dyn_hair, loc=<4.6828 -3.8041 2.2294>, rot=<-0.4982 0.5086 0.1844 -0.6776>}
 dyn_hair_v {parent=dyn_hair, loc=<4.621 -3.9231 -2.0617>, rot=<0.222 -0.3566 0.2839 -0.862>}
 dyn_hair_w {parent=dyn_hair, loc=<2.0797 1.8722 5.7819>, rot=<0.1588 0.466 0.2433 -0.8357>}
 dyn_hair_x {parent=dyn_hair, loc=<4.6613 -1.3702 3.7053>, rot=<-0.0659 0.5764 0.4772 -0.66>}
 dyn_hair_u {parent=dyn_hair, loc=<-0.186 -4.172 7.3626>, rot=<-0.2761 0.5833 0.4104 -0.6443>}
 dyn_hair_z {parent=dyn_hair, loc=<-1.921 -2.8621 7.0349>, rot=<-0.1495 0.7239 0.201 -0.6428>}
 dyn_hair_zz {parent=dyn_hair, loc=<-1.6387 -8.1444 5.4482>, rot=<0.3579 -0.7623 -0.2463 0.4797>}
 dyn_hair_xx {parent=dyn_hair, loc=<1.6101 -1.8275 7.7108>, rot=<0.0851 -0.8769 -0.182 0.4366>}
 dyn_hair_uu {parent=dyn_hair, loc=<1.4612 5.3127 4.9581>, rot=<-0.0958 0.5158 -0.6764 -0.5171>}
 dyn_hair_ww {parent=dyn_hair, loc=<2.1452 2.7922 3.8407>, rot=<-0.5507 0.4163 -0.3741 -0.6192>}
 dyn_hair_kk {parent=dyn_hair, loc=<2.8746 -0.7026 -5.8735>, rot=<0.5724 -0.5575 -0.5101 -0.3184>}
 dyn_hair_aa {parent=dyn_hair, loc=<-0.9131 -2.3831 -8.1848>, rot=<-0.5304 0.568 0.6125 0.1447>}
 dyn_hair_bb {parent=dyn_hair, loc=<0.6219 1.1498 -5.5451>, rot=<-0.4559 0.5345 0.6748 0.2259>}
 dyn_hair_dd {parent=dyn_hair, loc=<2.2985 -3.5124 -5.3641>, rot=<-0.5985 0.1616 0.7465 -0.2418>}
 dyn_hair_cc {parent=dyn_hair, loc=<0.296 -4.5116 -7.0463>, rot=<-0.5523 0.3432 0.72 -0.2424>}
 dyn_hair_ss {parent=dyn_hair, loc=<4.9607 1.774 -3.3372>, rot=<0.2687 -0.4286 0.2944 -0.8108>}
 clavicle_l {parent=spine_05, loc=<4.3575 -0.1466 -2.3223>, rot=<0.0465 -0.7289 -0.009 -0.6829>}
 clavicle_r {parent=spine_05, loc=<4.3575 -0.1466 2.3223>, rot=<0.7289 0.0465 -0.6829 0.009>}
]
variables [
 skydiving {type=bool, default=False}
 speed scale {type=double, default=0}
]
