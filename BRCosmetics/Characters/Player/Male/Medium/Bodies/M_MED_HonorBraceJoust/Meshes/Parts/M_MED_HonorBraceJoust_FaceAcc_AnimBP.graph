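# Post-copy AnimBP for the HonorBraceJoust male FaceAcc mesh, adding secondary hair motion.
# The parent mesh's pose feeds a Control Rig, a slight neck_01 scale-up, then 36 serial, identically sprung, angle-limited AnimDynamics nodes on dyn_hair bones.
# Each update derives the rig's skydiving flag and speed scale from float, parachute and dive states.
# It also sets ScaleFaceAcc at LOD 1+, which nothing shown reads.
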AnimBlueprint /BRCosmetics/Characters/Player/Male/Medium/Bodies/M_MED_HonorBraceJoust/Meshes/Parts/M_MED_HonorBraceJoust_FaceAcc_AnimBP
target_skeleton=&/M_MED_HonorBraceJoust_FaceAcc_Skeleton
graphs [
 AnimGraph {nodes [
   n1 = Control Rig!(
    Source=Copy Pose From Mesh() {bUseAttachedParent=true}
    Alpha
    speed scale:double=Get Hair Speed Scale() -> (Hair Speed Scale:double)
    skydiving:bool=Get Skydiving Cr() -> (Skydiving Cr:bool)
   ) -> (Pose) {controlRigAssetReference {blueprintRigClass=&/M_MED_HonorBraceJoust_FaceAcc_CtrlRig.M_MED_HonorBraceJoust_FaceAcc_CtrlRig_C}}
   n2 = "Transform (Modify) Bone - Bone: neck_01"(ComponentPose=Local To Component(LocalPose=@n1), Alpha=1, Translation="0,0,0", Rotation="0,0,0", Scale="1,1.02,1.02") {
    scaleMode=BMM_Additive
    scaleSpace=BCS_BoneSpace
    translationMode=BMM_Additive
    translationSpace=BCS_BoneSpace
    alphaBoolBlend {blendOption=Linear}
    boneToModify {boneName=neck_01}
    scale {y=1.02, z=1.02}
   }
   n3 = Anim Dynamics - Bone: dyn_hair_a(ComponentPose=@n2, Alpha=1, ExternalForce="0,0,0") {
    angularBiasOverride=0.2
    angularDampingOverride=0.9995
    angularSpringConstant=20000
    bAngularSpring=true
    bOverrideAngularBias=true
    bOverrideAngularDamping=true
    bUseGravityOverride=true
    gravityScale=0.35
    alphaBoolBlend {blendOption=Linear}
    boundBone {boneName=dyn_hair_a}
    componentAppliedLinearAccClamp=<200000 200000 200000>
    componentLinearAccScale=<1 1 1>
    componentLinearVelScale=<1 1 1>
    gravityOverride {z=-980}
    physicsBodyDefinitions [{
      boundBone {boneName=dyn_hair_a}
      boxExtents=<40 40 40>
      constraintSetup {angularLimitsMax=<7.5 20 20>, angularLimitsMin=<-7.5 -20 -20>, angularTarget=<1 0 0>}
      localJointOffset=<1 0 0>
     }]
    relativeSpaceBone {boneName=head}
   }
   n4 = Anim Dynamics - Bone: dyn_hair_k(ComponentPose=@n3, Alpha=1, ExternalForce="0,0,0") {
    angularBiasOverride=0.2
    angularDampingOverride=0.9995
    angularSpringConstant=20000
    bAngularSpring=true
    bOverrideAngularBias=true
    bOverrideAngularDamping=true
    bUseGravityOverride=true
    gravityScale=0.35
    alphaBoolBlend {blendOption=Linear}
    boundBone {boneName=dyn_hair_k}
    componentAppliedLinearAccClamp=<200000 200000 200000>
    componentLinearAccScale=<1 1 1>
    componentLinearVelScale=<1 1 1>
    gravityOverride {z=-980}
    physicsBodyDefinitions [{
      boundBone {boneName=dyn_hair_k}
      boxExtents=<40 50 50>
      constraintSetup {angularLimitsMax=<7.5 20 20>, angularLimitsMin=<-7.5 -20 -20>, angularTarget=<1 0 0>}
      localJointOffset=<1 0 0>
     }]
    relativeSpaceBone {boneName=head}
   }
   n5 = Anim Dynamics - Bone: dyn_hair_y(ComponentPose=@n4, Alpha=1, ExternalForce="0,0,0") {
    angularBiasOverride=0.2
    angularDampingOverride=0.9995
    angularSpringConstant=20000
    bAngularSpring=true
    bOverrideAngularBias=true
    bOverrideAngularDamping=true
    bUseGravityOverride=true
    gravityScale=0.35
    alphaBoolBlend {blendOption=Linear}
    boundBone {boneName=dyn_hair_y}
    componentAppliedLinearAccClamp=<200000 200000 200000>
    componentLinearAccScale=<1 1 1>
    componentLinearVelScale=<1 1 1>
    gravityOverride {z=-980}
    physicsBodyDefinitions [{
      boundBone {boneName=dyn_hair_y}
      boxExtents=<40 40 40>
      constraintSetup {angularLimitsMax=<7.5 20 20>, angularLimitsMin=<-7.5 -20 -20>, angularTarget=<1 0 0>}
      localJointOffset=<1 0 0>
     }]
    relativeSpaceBone {boneName=head}
   }
   n6 = Anim Dynamics - Bone: dyn_hair_b(ComponentPose=@n5, Alpha=1, ExternalForce="0,0,0") {
    angularBiasOverride=0.2
    angularDampingOverride=0.9995
    angularSpringConstant=20000
    bAngularSpring=true
    bOverrideAngularBias=true
    bOverrideAngularDamping=true
    bUseGravityOverride=true
    gravityScale=0.35
    alphaBoolBlend {blendOption=Linear}
    boundBone {boneName=dyn_hair_b}
    componentAppliedLinearAccClamp=<200000 200000 200000>
    componentLinearAccScale=<1 1 1>
    componentLinearVelScale=<1 1 1>
    gravityOverride {z=-980}
    physicsBodyDefinitions [{
      boundBone {boneName=dyn_hair_b}
      boxExtents=<40 40 40>
      constraintSetup {angularLimitsMax=<7.5 20 20>, angularLimitsMin=<-7.5 -20 -20>, angularTarget=<1 0 0>}
      localJointOffset=<1 0 0>
     }]
    relativeSpaceBone {boneName=head}
   }
   n7 = Anim Dynamics - Bone: dyn_hair_l(ComponentPose=@n6, Alpha=1, ExternalForce="0,0,0") {
    angularBiasOverride=0.2
    angularDampingOverride=0.9995
    angularSpringConstant=20000
    bAngularSpring=true
    bOverrideAngularBias=true
    bOverrideAngularDamping=true
    bUseGravityOverride=true
    gravityScale=0.35
    alphaBoolBlend {blendOption=Linear}
    boundBone {boneName=dyn_hair_l}
    componentAppliedLinearAccClamp=<200000 200000 200000>
    componentLinearAccScale=<1 1 1>
    componentLinearVelScale=<1 1 1>
    gravityOverride {z=-980}
    physicsBodyDefinitions [{
      boundBone {boneName=dyn_hair_l}
      boxExtents=<40 50 50>
      constraintSetup {angularLimitsMax=<7.5 20 20>, angularLimitsMin=<-7.5 -20 -20>, angularTarget=<1 0 0>}
      localJointOffset=<1 0 0>
     }]
    relativeSpaceBone {boneName=head}
   }
   n8 = Anim Dynamics - Bone: dyn_hair_v(ComponentPose=@n7, Alpha=1, ExternalForce="0,0,0") {
    angularBiasOverride=0.2
    angularDampingOverride=0.9995
    angularSpringConstant=20000
    bAngularSpring=true
    bOverrideAngularBias=true
    bOverrideAngularDamping=true
    bUseGravityOverride=true
    gravityScale=0.35
    alphaBoolBlend {blendOption=Linear}
    boundBone {boneName=dyn_hair_v}
    componentAppliedLinearAccClamp=<200000 200000 200000>
    componentLinearAccScale=<1 1 1>
    componentLinearVelScale=<1 1 1>
    gravityOverride {z=-980}
    physicsBodyDefinitions [{
      boundBone {boneName=dyn_hair_v}
      boxExtents=<40 40 40>
      constraintSetup {angularLimitsMax=<7.5 20 20>, angularLimitsMin=<-7.5 -20 -20>, angularTarget=<1 0 0>}
      localJointOffset=<1 0 0>
     }]
    relativeSpaceBone {boneName=head}
   }
   n9 = Anim Dynamics - Bone: dyn_hair_c(ComponentPose=@n8, Alpha=1, ExternalForce="0,0,0") {
    angularBiasOverride=0.2
    angularDampingOverride=0.9995
    angularSpringConstant=20000
    bAngularSpring=true
    bOverrideAngularBias=true
    bOverrideAngularDamping=true
    bUseGravityOverride=true
    gravityScale=0.35
    alphaBoolBlend {blendOption=Linear}
    boundBone {boneName=dyn_hair_c}
    componentAppliedLinearAccClamp=<200000 200000 200000>
    componentLinearAccScale=<1 1 1>
    componentLinearVelScale=<1 1 1>
    gravityOverride {z=-980}
    physicsBodyDefinitions [{
      boundBone {boneName=dyn_hair_c}
      boxExtents=<40 40 40>
      constraintSetup {angularLimitsMax=<7.5 20 20>, angularLimitsMin=<-7.5 -20 -20>, angularTarget=<1 0 0>}
      localJointOffset=<1 0 0>
     }]
    relativeSpaceBone {boneName=head}
   }
   n10 = Anim Dynamics - Bone: dyn_hair_m(ComponentPose=@n9, Alpha=1, ExternalForce="0,0,0") {
    angularBiasOverride=0.2
    angularDampingOverride=0.9995
    angularSpringConstant=20000
    bAngularSpring=true
    bOverrideAngularBias=true
    bOverrideAngularDamping=true
    bUseGravityOverride=true
    gravityScale=0.35
    alphaBoolBlend {blendOption=Linear}
    boundBone {boneName=dyn_hair_m}
    componentAppliedLinearAccClamp=<200000 200000 200000>
    componentLinearAccScale=<1 1 1>
    componentLinearVelScale=<1 1 1>
    gravityOverride {z=-980}
    physicsBodyDefinitions [{
      boundBone {boneName=dyn_hair_m}
      boxExtents=<40 50 50>
      constraintSetup {angularLimitsMax=<7.5 20 20>, angularLimitsMin=<-7.5 -20 -20>, angularTarget=<1 0 0>}
      localJointOffset=<1 0 0>
     }]
    relativeSpaceBone {boneName=head}
   }
   n11 = Anim Dynamics - Bone: dyn_hair_w(ComponentPose=@n10, Alpha=1, ExternalForce="0,0,0") {
    angularBiasOverride=0.2
    angularDampingOverride=0.9995
    angularSpringConstant=20000
    bAngularSpring=true
    bOverrideAngularBias=true
    bOverrideAngularDamping=true
    bUseGravityOverride=true
    gravityScale=0.35
    alphaBoolBlend {blendOption=Linear}
    boundBone {boneName=dyn_hair_w}
    componentAppliedLinearAccClamp=<200000 200000 200000>
    componentLinearAccScale=<1 1 1>
    componentLinearVelScale=<1 1 1>
    gravityOverride {z=-980}
    physicsBodyDefinitions [{
      boundBone {boneName=dyn_hair_w}
      boxExtents=<40 40 40>
      constraintSetup {angularLimitsMax=<7.5 20 20>, angularLimitsMin=<-7.5 -20 -20>, angularTarget=<1 0 0>}
      localJointOffset=<1 0 0>
     }]
    relativeSpaceBone {boneName=head}
   }
   n12 = Anim Dynamics - Bone: dyn_hair_d(ComponentPose=@n11, Alpha=1, ExternalForce="0,0,0") {
    angularBiasOverride=0.2
    angularDampingOverride=0.9995
    angularSpringConstant=20000
    bAngularSpring=true
    bOverrideAngularBias=true
    bOverrideAngularDamping=true
    bUseGravityOverride=true
    gravityScale=0.35
    alphaBoolBlend {blendOption=Linear}
    boundBone {boneName=dyn_hair_d}
    componentAppliedLinearAccClamp=<200000 200000 200000>
    componentLinearAccScale=<1 1 1>
    componentLinearVelScale=<1 1 1>
    gravityOverride {z=-980}
    physicsBodyDefinitions [{
      boundBone {boneName=dyn_hair_d}
      boxExtents=<40 40 40>
      constraintSetup {angularLimitsMax=<7.5 20 20>, angularLimitsMin=<-7.5 -20 -20>, angularTarget=<1 0 0>}
      localJointOffset=<1 0 0>
     }]
    relativeSpaceBone {boneName=head}
   }
   n13 = Anim Dynamics - Bone: dyn_hair_n(ComponentPose=@n12, Alpha=1, ExternalForce="0,0,0") {
    angularBiasOverride=0.2
    angularDampingOverride=0.9995
    angularSpringConstant=20000
    bAngularSpring=true
    bOverrideAngularBias=true
    bOverrideAngularDamping=true
    bUseGravityOverride=true
    gravityScale=0.35
    alphaBoolBlend {blendOption=Linear}
    boundBone {boneName=dyn_hair_n}
    componentAppliedLinearAccClamp=<200000 200000 200000>
    componentLinearAccScale=<1 1 1>
    componentLinearVelScale=<1 1 1>
    gravityOverride {z=-980}
    physicsBodyDefinitions [{
      boundBone {boneName=dyn_hair_n}
      boxExtents=<40 40 40>
      constraintSetup {angularLimitsMax=<7.5 20 20>, angularLimitsMin=<-7.5 -20 -20>, angularTarget=<1 0 0>}
      localJointOffset=<1 0 0>
     }]
    relativeSpaceBone {boneName=head}
   }
   n14 = Anim Dynamics - Bone: dyn_hair_x(ComponentPose=@n13, Alpha=1, ExternalForce="0,0,0") {
    angularBiasOverride=0.2
    angularDampingOverride=0.9995
    angularSpringConstant=20000
    bAngularSpring=true
    bOverrideAngularBias=true
    bOverrideAngularDamping=true
    bUseGravityOverride=true
    gravityScale=0.35
    alphaBoolBlend {blendOption=Linear}
    boundBone {boneName=dyn_hair_x}
    componentAppliedLinearAccClamp=<200000 200000 200000>
    componentLinearAccScale=<1 1 1>
    componentLinearVelScale=<1 1 1>
    gravityOverride {z=-980}
    physicsBodyDefinitions [{
      boundBone {boneName=dyn_hair_x}
      boxExtents=<40 40 40>
      constraintSetup {angularLimitsMax=<7.5 20 20>, angularLimitsMin=<-7.5 -20 -20>, angularTarget=<1 0 0>}
      localJointOffset=<1 0 0>
     }]
    relativeSpaceBone {boneName=head}
   }
   n15 = Anim Dynamics - Bone: dyn_hair_e(ComponentPose=@n14, Alpha=1, ExternalForce="0,0,0") {
    angularBiasOverride=0.2
    angularDampingOverride=0.9995
    angularSpringConstant=20000
    bAngularSpring=true
    bOverrideAngularBias=true
    bOverrideAngularDamping=true
    bUseGravityOverride=true
    gravityScale=0.35
    alphaBoolBlend {blendOption=Linear}
    boundBone {boneName=dyn_hair_e}
    componentAppliedLinearAccClamp=<200000 200000 200000>
    componentLinearAccScale=<1 1 1>
    componentLinearVelScale=<1 1 1>
    gravityOverride {z=-980}
    physicsBodyDefinitions [{
      boundBone {boneName=dyn_hair_e}
      boxExtents=<40 50 50>
      constraintSetup {angularLimitsMax=<7.5 20 20>, angularLimitsMin=<-7.5 -20 -20>, angularTarget=<1 0 0>}
      localJointOffset=<1 0 0>
     }]
    relativeSpaceBone {boneName=head}
   }
   n16 = Anim Dynamics - Bone: dyn_hair_o(ComponentPose=@n15, Alpha=1, ExternalForce="0,0,0") {
    angularBiasOverride=0.2
    angularDampingOverride=0.9995
    angularSpringConstant=20000
    bAngularSpring=true
    bOverrideAngularBias=true
    bOverrideAngularDamping=true
    bUseGravityOverride=true
    gravityScale=0.35
    alphaBoolBlend {blendOption=Linear}
    boundBone {boneName=dyn_hair_o}
    componentAppliedLinearAccClamp=<200000 200000 200000>
    componentLinearAccScale=<1 1 1>
    componentLinearVelScale=<1 1 1>
    gravityOverride {z=-980}
    physicsBodyDefinitions [{
      boundBone {boneName=dyn_hair_o}
      boxExtents=<40 40 40>
      constraintSetup {angularLimitsMax=<7.5 20 20>, angularLimitsMin=<-7.5 -20 -20>, angularTarget=<1 0 0>}
      localJointOffset=<1 0 0>
     }]
    relativeSpaceBone {boneName=head}
   }
   n17 = Anim Dynamics - Bone: dyn_hair_u(ComponentPose=@n16, Alpha=1, ExternalForce="0,0,0") {
    angularBiasOverride=0.2
    angularDampingOverride=0.9995
    angularSpringConstant=20000
    bAngularSpring=true
    bOverrideAngularBias=true
    bOverrideAngularDamping=true
    bUseGravityOverride=true
    gravityScale=0.35
    alphaBoolBlend {blendOption=Linear}
    boundBone {boneName=dyn_hair_u}
    componentAppliedLinearAccClamp=<200000 200000 200000>
    componentLinearAccScale=<1 1 1>
    componentLinearVelScale=<1 1 1>
    gravityOverride {z=-980}
    physicsBodyDefinitions [{
      boundBone {boneName=dyn_hair_u}
      boxExtents=<40 40 40>
      constraintSetup {angularLimitsMax=<7.5 20 20>, angularLimitsMin=<-7.5 -20 -20>, angularTarget=<1 0 0>}
      localJointOffset=<1 0 0>
     }]
    relativeSpaceBone {boneName=head}
   }
   n18 = Anim Dynamics - Bone: dyn_hair_f(ComponentPose=@n17, Alpha=1, ExternalForce="0,0,0") {
    angularBiasOverride=0.2
    angularDampingOverride=0.9995
    angularSpringConstant=20000
    bAngularSpring=true
    bOverrideAngularBias=true
    bOverrideAngularDamping=true
    bUseGravityOverride=true
    gravityScale=0.35
    alphaBoolBlend {blendOption=Linear}
    boundBone {boneName=dyn_hair_f}
    componentAppliedLinearAccClamp=<200000 200000 200000>
    componentLinearAccScale=<1 1 1>
    componentLinearVelScale=<1 1 1>
    gravityOverride {z=-980}
    physicsBodyDefinitions [{
      boundBone {boneName=dyn_hair_f}
      boxExtents=<40 50 50>
      constraintSetup {angularLimitsMax=<7.5 20 20>, angularLimitsMin=<-7.5 -20 -20>, angularTarget=<1 0 0>}
      localJointOffset=<1 0 0>
     }]
    relativeSpaceBone {boneName=head}
   }
   n19 = Anim Dynamics - Bone: dyn_hair_p(ComponentPose=@n18, Alpha=1, ExternalForce="0,0,0") {
    angularBiasOverride=0.2
    angularDampingOverride=0.9995
    angularSpringConstant=20000
    bAngularSpring=true
    bOverrideAngularBias=true
    bOverrideAngularDamping=true
    bUseGravityOverride=true
    gravityScale=0.35
    alphaBoolBlend {blendOption=Linear}
    boundBone {boneName=dyn_hair_p}
    componentAppliedLinearAccClamp=<200000 200000 200000>
    componentLinearAccScale=<1 1 1>
    componentLinearVelScale=<1 1 1>
    gravityOverride {z=-980}
    physicsBodyDefinitions [{
      boundBone {boneName=dyn_hair_p}
      boxExtents=<40 40 40>
      constraintSetup {angularLimitsMax=<7.5 20 20>, angularLimitsMin=<-7.5 -20 -20>, angularTarget=<1 0 0>}
      localJointOffset=<1 0 0>
     }]
    relativeSpaceBone {boneName=head}
   }
   n20 = Anim Dynamics - Bone: dyn_hair_z(ComponentPose=@n19, Alpha=1, ExternalForce="0,0,0") {
    angularBiasOverride=0.2
    angularDampingOverride=0.9995
    angularSpringConstant=20000
    bAngularSpring=true
    bOverrideAngularBias=true
    bOverrideAngularDamping=true
    bUseGravityOverride=true
    gravityScale=0.35
    alphaBoolBlend {blendOption=Linear}
    boundBone {boneName=dyn_hair_z}
    componentAppliedLinearAccClamp=<200000 200000 200000>
    componentLinearAccScale=<1 1 1>
    componentLinearVelScale=<1 1 1>
    gravityOverride {z=-980}
    physicsBodyDefinitions [{
      boundBone {boneName=dyn_hair_z}
      boxExtents=<40 40 40>
      constraintSetup {angularLimitsMax=<7.5 20 20>, angularLimitsMin=<-7.5 -20 -20>, angularTarget=<1 0 0>}
      localJointOffset=<1 0 0>
     }]
    relativeSpaceBone {boneName=head}
   }
   n21 = Anim Dynamics - Bone: dyn_hair_g(ComponentPose=@n20, Alpha=1, ExternalForce="0,0,0") {
    angularBiasOverride=0.2
    angularDampingOverride=0.9995
    angularSpringConstant=20000
    bAngularSpring=true
    bOverrideAngularBias=true
    bOverrideAngularDamping=true
    bUseGravityOverride=true
    gravityScale=0.35
    alphaBoolBlend {blendOption=Linear}
    boundBone {boneName=dyn_hair_g}
    componentAppliedLinearAccClamp=<200000 200000 200000>
    componentLinearAccScale=<1 1 1>
    componentLinearVelScale=<1 1 1>
    gravityOverride {z=-980}
    physicsBodyDefinitions [{
      boundBone {boneName=dyn_hair_g}
      boxExtents=<40 50 50>
      constraintSetup {angularLimitsMax=<7.5 20 20>, angularLimitsMin=<-7.5 -20 -20>, angularTarget=<1 0 0>}
      localJointOffset=<1 0 0>
     }]
    relativeSpaceBone {boneName=head}
   }
   n22 = Anim Dynamics - Bone: dyn_hair_q(ComponentPose=@n21, Alpha=1, ExternalForce="0,0,0") {
    angularBiasOverride=0.2
    angularDampingOverride=0.9995
    angularSpringConstant=20000
    bAngularSpring=true
    bOverrideAngularBias=true
    bOverrideAngularDamping=true
    bUseGravityOverride=true
    gravityScale=0.35
    alphaBoolBlend {blendOption=Linear}
    boundBone {boneName=dyn_hair_q}
    componentAppliedLinearAccClamp=<200000 200000 200000>
    componentLinearAccScale=<1 1 1>
    componentLinearVelScale=<1 1 1>
    gravityOverride {z=-980}
    physicsBodyDefinitions [{
      boundBone {boneName=dyn_hair_q}
      boxExtents=<40 40 40>
      constraintSetup {angularLimitsMax=<7.5 20 20>, angularLimitsMin=<-7.5 -20 -20>, angularTarget=<1 0 0>}
      localJointOffset=<1 0 0>
     }]
    relativeSpaceBone {boneName=head}
   }
   n23 = Anim Dynamics - Bone: dyn_hair_zz(ComponentPose=@n22, Alpha=1, ExternalForce="0,0,0") {
    angularBiasOverride=0.2
    angularDampingOverride=0.9995
    angularSpringConstant=20000
    bAngularSpring=true
    bOverrideAngularBias=true
    bOverrideAngularDamping=true
    bUseGravityOverride=true
    gravityScale=0.35
    alphaBoolBlend {blendOption=Linear}
    boundBone {boneName=dyn_hair_zz}
    componentAppliedLinearAccClamp=<200000 200000 200000>
    componentLinearAccScale=<1 1 1>
    componentLinearVelScale=<1 1 1>
    gravityOverride {z=-980}
    physicsBodyDefinitions [{
      boundBone {boneName=dyn_hair_zz}
      boxExtents=<40 40 40>
      constraintSetup {angularLimitsMax=<7.5 20 20>, angularLimitsMin=<-7.5 -20 -20>, angularTarget=<1 0 0>}
      localJointOffset=<1 0 0>
     }]
    relativeSpaceBone {boneName=head}
   }
   n24 = Anim Dynamics - Bone: dyn_hair_h(ComponentPose=@n23, Alpha=1, ExternalForce="0,0,0") {
    angularBiasOverride=0.2
    angularDampingOverride=0.9995
    angularSpringConstant=20000
    bAngularSpring=true
    bOverrideAngularBias=true
    bOverrideAngularDamping=true
    bUseGravityOverride=true
    gravityScale=0.35
    alphaBoolBlend {blendOption=Linear}
    boundBone {boneName=dyn_hair_h}
    componentAppliedLinearAccClamp=<200000 200000 200000>
    componentLinearAccScale=<1 1 1>
    componentLinearVelScale=<1 1 1>
    gravityOverride {z=-980}
    physicsBodyDefinitions [{
      boundBone {boneName=dyn_hair_h}
      boxExtents=<40 50 50>
      constraintSetup {angularLimitsMax=<7.5 20 20>, angularLimitsMin=<-7.5 -20 -20>, angularTarget=<1 0 0>}
      localJointOffset=<1 0 0>
     }]
    relativeSpaceBone {boneName=head}
   }
   n25 = Anim Dynamics - Bone: dyn_hair_r(ComponentPose=@n24, Alpha=1, ExternalForce="0,0,0") {
    angularBiasOverride=0.2
    angularDampingOverride=0.9995
    angularSpringConstant=20000
    bAngularSpring=true
    bOverrideAngularBias=true
    bOverrideAngularDamping=true
    bUseGravityOverride=true
    gravityScale=0.35
    alphaBoolBlend {blendOption=Linear}
    boundBone {boneName=dyn_hair_r}
    componentAppliedLinearAccClamp=<200000 200000 200000>
    componentLinearAccScale=<1 1 1>
    componentLinearVelScale=<1 1 1>
    gravityOverride {z=-980}
    physicsBodyDefinitions [{
      boundBone {boneName=dyn_hair_r}
      boxExtents=<40 40 40>
      constraintSetup {angularLimitsMax=<7.5 20 20>, angularLimitsMin=<-7.5 -20 -20>, angularTarget=<1 0 0>}
      localJointOffset=<1 0 0>
     }]
    relativeSpaceBone {boneName=head}
   }
   n26 = Anim Dynamics - Bone: dyn_hair_xx(ComponentPose=@n25, Alpha=1, ExternalForce="0,0,0") {
    angularBiasOverride=0.2
    angularDampingOverride=0.9995
    angularSpringConstant=20000
    bAngularSpring=true
    bOverrideAngularBias=true
    bOverrideAngularDamping=true
    bUseGravityOverride=true
    gravityScale=0.35
    alphaBoolBlend {blendOption=Linear}
    boundBone {boneName=dyn_hair_xx}
    componentAppliedLinearAccClamp=<200000 200000 200000>
    componentLinearAccScale=<1 1 1>
    componentLinearVelScale=<1 1 1>
    gravityOverride {z=-980}
    physicsBodyDefinitions [{
      boundBone {boneName=dyn_hair_xx}
      boxExtents=<40 40 40>
      constraintSetup {angularLimitsMax=<7.5 20 20>, angularLimitsMin=<-7.5 -20 -20>, angularTarget=<1 0 0>}
      localJointOffset=<1 0 0>
     }]
    relativeSpaceBone {boneName=head}
   }
   n27 = Anim Dynamics - Bone: dyn_hair_i(ComponentPose=@n26, Alpha=1, ExternalForce="0,0,0") {
    angularBiasOverride=0.2
    angularDampingOverride=0.9995
    angularSpringConstant=20000
    bAngularSpring=true
    bOverrideAngularBias=true
    bOverrideAngularDamping=true
    bUseGravityOverride=true
    gravityScale=0.35
    alphaBoolBlend {blendOption=Linear}
    boundBone {boneName=dyn_hair_i}
    componentAppliedLinearAccClamp=<200000 200000 200000>
    componentLinearAccScale=<1 1 1>
    componentLinearVelScale=<1 1 1>
    gravityOverride {z=-980}
    physicsBodyDefinitions [{
      boundBone {boneName=dyn_hair_i}
      boxExtents=<40 50 50>
      constraintSetup {angularLimitsMax=<7.5 20 20>, angularLimitsMin=<-7.5 -20 -20>, angularTarget=<1 0 0>}
      localJointOffset=<1 0 0>
     }]
    relativeSpaceBone {boneName=head}
   }
   n28 = Anim Dynamics - Bone: dyn_hair_s(ComponentPose=@n27, Alpha=1, ExternalForce="0,0,0") {
    angularBiasOverride=0.2
    angularDampingOverride=0.9995
    angularSpringConstant=20000
    bAngularSpring=true
    bOverrideAngularBias=true
    bOverrideAngularDamping=true
    bUseGravityOverride=true
    gravityScale=0.35
    alphaBoolBlend {blendOption=Linear}
    boundBone {boneName=dyn_hair_s}
    componentAppliedLinearAccClamp=<200000 200000 200000>
    componentLinearAccScale=<1 1 1>
    componentLinearVelScale=<1 1 1>
    gravityOverride {z=-980}
    physicsBodyDefinitions [{
      boundBone {boneName=dyn_hair_s}
      boxExtents=<40 40 40>
      constraintSetup {angularLimitsMax=<7.5 20 20>, angularLimitsMin=<-7.5 -20 -20>, angularTarget=<1 0 0>}
      localJointOffset=<1 0 0>
     }]
    relativeSpaceBone {boneName=head}
   }
   n29 = Anim Dynamics - Bone: dyn_hair_uu(ComponentPose=@n28, Alpha=1, ExternalForce="0,0,0") {
    angularBiasOverride=0.2
    angularDampingOverride=0.9995
    angularSpringConstant=20000
    bAngularSpring=true
    bOverrideAngularBias=true
    bOverrideAngularDamping=true
    bUseGravityOverride=true
    gravityScale=0.35
    alphaBoolBlend {blendOption=Linear}
    boundBone {boneName=dyn_hair_uu}
    componentAppliedLinearAccClamp=<200000 200000 200000>
    componentLinearAccScale=<1 1 1>
    componentLinearVelScale=<1 1 1>
    gravityOverride {z=-980}
    physicsBodyDefinitions [{
      boundBone {boneName=dyn_hair_uu}
      boxExtents=<40 40 40>
      constraintSetup {angularLimitsMax=<7.5 20 20>, angularLimitsMin=<-7.5 -20 -20>, angularTarget=<1 0 0>}
      localJointOffset=<1 0 0>
     }]
    relativeSpaceBone {boneName=head}
   }
   n30 = Anim Dynamics - Bone: dyn_hair_j(ComponentPose=@n29, Alpha=1, ExternalForce="0,0,0") {
    angularBiasOverride=0.2
    angularDampingOverride=0.9995
    angularSpringConstant=20000
    bAngularSpring=true
    bOverrideAngularBias=true
    bOverrideAngularDamping=true
    bUseGravityOverride=true
    gravityScale=0.35
    alphaBoolBlend {blendOption=Linear}
    boundBone {boneName=dyn_hair_j}
    componentAppliedLinearAccClamp=<200000 200000 200000>
    componentLinearAccScale=<1 1 1>
    componentLinearVelScale=<1 1 1>
    gravityOverride {z=-980}
    physicsBodyDefinitions [{
      boundBone {boneName=dyn_hair_j}
      boxExtents=<40 50 50>
      constraintSetup {angularLimitsMax=<7.5 20 20>, angularLimitsMin=<-7.5 -20 -20>, angularTarget=<1 0 0>}
      localJointOffset=<1 0 0>
     }]
    relativeSpaceBone {boneName=head}
   }
   n31 = Anim Dynamics - Bone: dyn_hair_t(ComponentPose=@n30, Alpha=1, ExternalForce="0,0,0") {
    angularBiasOverride=0.2
    angularDampingOverride=0.9995
    angularSpringConstant=20000
    bAngularSpring=true
    bOverrideAngularBias=true
    bOverrideAngularDamping=true
    bUseGravityOverride=true
    gravityScale=0.35
    alphaBoolBlend {blendOption=Linear}
    boundBone {boneName=dyn_hair_t}
    componentAppliedLinearAccClamp=<200000 200000 200000>
    componentLinearAccScale=<1 1 1>
    componentLinearVelScale=<1 1 1>
    gravityOverride {z=-980}
    physicsBodyDefinitions [{
      boundBone {boneName=dyn_hair_t}
      boxExtents=<40 40 40>
      constraintSetup {angularLimitsMax=<7.5 20 20>, angularLimitsMin=<-7.5 -20 -20>, angularTarget=<1 0 0>}
      localJointOffset=<1 0 0>
     }]
    relativeSpaceBone {boneName=head}
   }
   n32 = Anim Dynamics - Bone: dyn_hair_ww(ComponentPose=@n31, Alpha=1, ExternalForce="0,0,0") {
    angularBiasOverride=0.2
    angularDampingOverride=0.9995
    angularSpringConstant=20000
    bAngularSpring=true
    bOverrideAngularBias=true
    bOverrideAngularDamping=true
    bUseGravityOverride=true
    gravityScale=0.35
    alphaBoolBlend {blendOption=Linear}
    boundBone {boneName=dyn_hair_ww}
    componentAppliedLinearAccClamp=<200000 200000 200000>
    componentLinearAccScale=<1 1 1>
    componentLinearVelScale=<1 1 1>
    gravityOverride {z=-980}
    physicsBodyDefinitions [{
      boundBone {boneName=dyn_hair_ww}
      boxExtents=<40 40 40>
      constraintSetup {angularLimitsMax=<7.5 20 20>, angularLimitsMin=<-7.5 -20 -20>, angularTarget=<1 0 0>}
      localJointOffset=<1 0 0>
     }]
    relativeSpaceBone {boneName=head}
   }
   n33 = Anim Dynamics - Bone: dyn_hair_kk(ComponentPose=@n32, Alpha=1, ExternalForce="0,0,0") {
    angularBiasOverride=0.2
    angularDampingOverride=0.9995
    angularSpringConstant=20000
    bAngularSpring=true
    bOverrideAngularBias=true
    bOverrideAngularDamping=true
    bUseGravityOverride=true
    gravityScale=0.35
    alphaBoolBlend {blendOption=Linear}
    boundBone {boneName=dyn_hair_kk}
    componentAppliedLinearAccClamp=<200000 200000 200000>
    componentLinearAccScale=<1 1 1>
    componentLinearVelScale=<1 1 1>
    gravityOverride {z=-980}
    physicsBodyDefinitions [{
      boundBone {boneName=dyn_hair_kk}
      boxExtents=<40 40 40>
      constraintSetup {angularLimitsMax=<7.5 20 20>, angularLimitsMin=<-7.5 -20 -20>, angularTarget=<1 0 0>}
      localJointOffset=<1 0 0>
     }]
    relativeSpaceBone {boneName=head}
   }
   n34 = Anim Dynamics - Bone: dyn_hair_aa(ComponentPose=@n33, Alpha=1, ExternalForce="0,0,0") {
    angularBiasOverride=0.2
    angularDampingOverride=0.9995
    angularSpringConstant=20000
    bAngularSpring=true
    bOverrideAngularBias=true
    bOverrideAngularDamping=true
    bUseGravityOverride=true
    gravityScale=0.35
    alphaBoolBlend {blendOption=Linear}
    boundBone {boneName=dyn_hair_aa}
    componentAppliedLinearAccClamp=<200000 200000 200000>
    componentLinearAccScale=<1 1 1>
    componentLinearVelScale=<1 1 1>
    gravityOverride {z=-980}
    physicsBodyDefinitions [{
      boundBone {boneName=dyn_hair_aa}
      boxExtents=<40 40 40>
      constraintSetup {angularLimitsMax=<7.5 20 20>, angularLimitsMin=<-7.5 -20 -20>, angularTarget=<1 0 0>}
      localJointOffset=<1 0 0>
     }]
    relativeSpaceBone {boneName=head}
   }
   n35 = Anim Dynamics - Bone: dyn_hair_bb(ComponentPose=@n34, Alpha=1, ExternalForce="0,0,0") {
    angularBiasOverride=0.2
    angularDampingOverride=0.9995
    angularSpringConstant=20000
    bAngularSpring=true
    bOverrideAngularBias=true
    bOverrideAngularDamping=true
    bUseGravityOverride=true
    gravityScale=0.35
    alphaBoolBlend {blendOption=Linear}
    boundBone {boneName=dyn_hair_bb}
    componentAppliedLinearAccClamp=<200000 200000 200000>
    componentLinearAccScale=<1 1 1>
    componentLinearVelScale=<1 1 1>
    gravityOverride {z=-980}
    physicsBodyDefinitions [{
      boundBone {boneName=dyn_hair_bb}
      boxExtents=<40 40 40>
      constraintSetup {angularLimitsMax=<7.5 20 20>, angularLimitsMin=<-7.5 -20 -20>, angularTarget=<1 0 0>}
      localJointOffset=<1 0 0>
     }]
    relativeSpaceBone {boneName=head}
   }
   n36 = Anim Dynamics - Bone: dyn_hair_dd(ComponentPose=@n35, Alpha=1, ExternalForce="0,0,0") {
    angularBiasOverride=0.2
    angularDampingOverride=0.9995
    angularSpringConstant=20000
    bAngularSpring=true
    bOverrideAngularBias=true
    bOverrideAngularDamping=true
    bUseGravityOverride=true
    gravityScale=0.35
    alphaBoolBlend {blendOption=Linear}
    boundBone {boneName=dyn_hair_dd}
    componentAppliedLinearAccClamp=<200000 200000 200000>
    componentLinearAccScale=<1 1 1>
    componentLinearVelScale=<1 1 1>
    gravityOverride {z=-980}
    physicsBodyDefinitions [{
      boundBone {boneName=dyn_hair_dd}
      boxExtents=<40 40 40>
      constraintSetup {angularLimitsMax=<7.5 20 20>, angularLimitsMin=<-7.5 -20 -20>, angularTarget=<1 0 0>}
      localJointOffset=<1 0 0>
     }]
    relativeSpaceBone {boneName=head}
   }
   n37 = Anim Dynamics - Bone: dyn_hair_cc(ComponentPose=@n36, Alpha=1, ExternalForce="0,0,0") {
    angularBiasOverride=0.2
    angularDampingOverride=0.9995
    angularSpringConstant=20000
    bAngularSpring=true
    bOverrideAngularBias=true
    bOverrideAngularDamping=true
    bUseGravityOverride=true
    gravityScale=0.35
    alphaBoolBlend {blendOption=Linear}
    boundBone {boneName=dyn_hair_cc}
    componentAppliedLinearAccClamp=<200000 200000 200000>
    componentLinearAccScale=<1 1 1>
    componentLinearVelScale=<1 1 1>
    gravityOverride {z=-980}
    physicsBodyDefinitions [{
      boundBone {boneName=dyn_hair_cc}
      boxExtents=<40 40 40>
      constraintSetup {angularLimitsMax=<7.5 20 20>, angularLimitsMin=<-7.5 -20 -20>, angularTarget=<1 0 0>}
      localJointOffset=<1 0 0>
     }]
    relativeSpaceBone {boneName=head}
   }
   n38 = Anim Dynamics - Bone: dyn_hair_ss(ComponentPose=@n37, Alpha=1, ExternalForce="0,0,0") {
    angularBiasOverride=0.2
    angularDampingOverride=0.9995
    angularSpringConstant=20000
    bAngularSpring=true
    bOverrideAngularBias=true
    bOverrideAngularDamping=true
    bUseGravityOverride=true
    gravityScale=0.35
    alphaBoolBlend {blendOption=Linear}
    boundBone {boneName=dyn_hair_ss}
    componentAppliedLinearAccClamp=<200000 200000 200000>
    componentLinearAccScale=<1 1 1>
    componentLinearVelScale=<1 1 1>
    gravityOverride {z=-980}
    physicsBodyDefinitions [{
      boundBone {boneName=dyn_hair_ss}
      boxExtents=<40 40 40>
      constraintSetup {angularLimitsMax=<7.5 20 20>, angularLimitsMin=<-7.5 -20 -20>, angularTarget=<1 0 0>}
      localJointOffset=<1 0 0>
     }]
    relativeSpaceBone {boneName=head}
   }
   Output Pose(Result=Component To Local(ComponentPose=@n38)) {layerGroup=None}
  ]}
 EventGraph {nodes [
   Event BlueprintUpdateAnimation()
   > n1 = Sequence()
   > Set ScaleFaceAcc(ScaleFaceAcc:bool="integer >= integer"(A=Get CurrentLOD(), B=1)) -> (Output_Get:bool)
   Set skydiving cr(
    execute=@n1.then_1
    skydiving cr:bool=OR Boolean!(A=Get bIsSkydiveFloating(), B=Get bIsParachuteOpened(), C:bool=Get bIsSkyDiving()) -> (ReturnValue)
   ) -> (Output_Get:bool)
   > n2 = Branch(Condition=Get bIsSkydiveFloating())
   > Set hair speed scale(hair speed scale:double=5) -> (Output_Get:double)
   n3 = Branch(execute=@n2.else, Condition=Get bIsParachuteOpened())
   > Set hair speed scale(hair speed scale=3)
   n4 = Branch(execute=@n3.else, Condition=Get bIsSkyDiving())
   > Set hair speed scale(hair speed scale=8)
   Set hair speed scale(execute=@n4.else, hair speed scale=0)
  ]}
]
variables [
 Hair Speed Scale {type=number, default=5}
 ScaleFaceAcc {type=boolean}
 Sim Space Settings {
  type=SimSpaceSettings
  default {
   dampingAlpha=1
   maxAngularAcceleration=10000
   maxAngularVelocity=10000
   maxLinearAcceleration=10000
   maxLinearVelocity=10000
   velocityScaleZ=1
   worldAlpha=0
   externalLinearDragV=<0 0 0>
  }
 }
 Skydiving Cr {type=boolean}
]
